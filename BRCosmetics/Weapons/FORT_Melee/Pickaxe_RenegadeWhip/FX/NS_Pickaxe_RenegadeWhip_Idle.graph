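# Niagara idle effect for the Renegade Whip pickaxe, drawing a purple-pink engine exhaust at a weapon socket.
# EngineFlames supplies waveform-flickered flame meshes, ExhaustGlow adds pulsing sprites and light, and Flare adds a delayed burst with its own light.
# GPU Embers spray in a cone with curl noise and stop spawning during swings.
# System-level AlphaFade scales flames and glow, while CameraDot drives view-dependent size, opacity and camera offset.
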NiagaraSystem /BRCosmetics/Weapons/FORT_Melee/Pickaxe_RenegadeWhip/FX/NS_Pickaxe_RenegadeWhip_Idle
dependencies {
 usedDataInterfaces [
  &NiagaraDataInterfaceCurve
  &NiagaraDataInterfaceColorCurve
  &NiagaraDataInterfaceSocketReader
  &NiagaraDataInterfaceCamera
 ]
 usedDynamicInputs [
  &DotProduct
  &ReturnCameraVector
  /Game/Effects/Niagara/DynamicInputs/Engine/FN_GetSystemAxis
 ]
 usedModules [
  &self:SystemSpawnScript.SystemScriptSource.SystemScriptGraph.NiagaraNodeAssignment_0.SetVariables_#1
  &SystemState
  /BRCosmetics/Characters/Player/Female/Medium/Bodies/F_MED_PiperShelf_Pearl/FX/TimeOfDayGlobal
  &self:SystemSpawnScript.SystemScriptSource.SystemScriptGraph.NiagaraNodeAssignment_1.SetVariables_#2
 ]
}
emitters {
 Embers {
  simTarget=GPUComputeSim
  emitterUpdateScript [
   &EmitterState()
   &SpawnRate(
    Spawn Probability=&MakeCustomFloatFromBool(Bool=$User.bIsSwinging, True Float=0, False Float=0.3)
    SpawnRate=&Multiply_Float(
     A=20
     B=&FloatFromCurve(FloatCurve=DataInterface() {propertyValues {Curve {keys [<at 0 0>,<at 0.3 1>,<at 0.5 1>]}}}, CurveIndex=$System.Age)
    )
    Use Spawn Probability=true
   ) {module=SpawnRate001}
  ]
  particleSpawnScript [
   /SetVariables(Particles.PerParticleRandom:NiagaraFloat=&RandomRangeFloat(Minimum=0, Maximum=1)) {isSetParameters=true}
   &InitializeParticle(
    Color Mode=::NewEnumerator0
    Lifetime=&RandomRangeFloat(Minimum=0.111, Maximum=0.25)
    Sprite Size=&Multiply_Vector2D(
     A=<0.5 1>
     B=&MakeVector2D(
      X=&FloatFromCurve(
       FloatCurve=DataInterface() {propertyValues {Curve {keys [<at 0 1>,<at 0.8 2 RCIM_Constant>,<at 0.95 3 RCIM_Constant>]}}}
       CurveIndex=$Particles.PerParticleRandom
      )
      Y=&FloatFromCurve(
       FloatCurve=DataInterface() {propertyValues {Curve {keys [<at 0 1>,<at 0.8 2 RCIM_Constant>,<at 0.95 3>]}}}
       CurveIndex=$Particles.PerParticleRandom
      )
     )
    )
    Sprite Size Mode=::NewEnumerator4
    UsePositionOffset=false
    Write Material Random=true
   )
   &SocketLocation(SamplePosOffset=false, Socket Index=2, Socket Reader=$System.Socket Reader, Socket Sampling Mode=::NewEnumerator3, VelocitySpeedLimit=false)
   &AddVelocity(
    Cone Angle=30
    Cone Axis=<0 0 1>
    Module.UseConeFalloff=false
    Module.UseVelDistribution=false
    Override Seed=false
    Velocity Mode=::NewEnumerator2
    Velocity Speed=&Multiply_Float(
     A=&RandomRangeFloat(Minimum=80, Maximum=300)
     B=&FloatFromCurve(FloatCurve=DataInterface() {propertyValues {Curve {keys [<at 0.3 0.3>,<at 0.8 1>]}}}, CurveIndex=$Engine.System.Age)
    )
   )
   &ShapeLocation(MaskSpawnGroup=false, Override Seed=false, Sphere Radius=1, UseOwnerScale=true)
   &CameraOffset()
  ]
  particleUpdateScript [
   &ParticleState()
   &Drag(
    Drag=&FloatFromCurve(
     FloatCurve=DataInterface() {propertyValues {Curve {keys [{tangentMode=RCTM_User, time=0, value=0}, {tangentMode=RCTM_User, time=1, value=1}]}}}
     Scale Curve=&RandomRangeFloat(Minimum=5, Maximum=50)
    )
    Use Linear Drag=true
    Use Rotational Drag=false
   )
   /Niagara/Modules/Update/Forces/CurlNoiseForce(
    Noise Frequency=0.15
    Noise Strength=&FloatFromCurve(
     FloatCurve=DataInterface() {propertyValues {Curve {keys [{tangentMode=RCTM_User, time=0, value=0}, {tangentMode=RCTM_User, time=1, value=1}]}}}
     Scale Curve=&Lerp_Float(A=5000, B=222, Alpha=$Particles.PerParticleRandom)
    )
    Pan Noise=true
    Pan Noise Field=<0 0 1>
   )
   &SolveForcesAndVelocity(Clamp Velocity=false, Limit Acceleration=false)
   &ScaleColor(
    Scale Mode=::NewEnumerator1
    Scale RGBA=&MakeVector4_FromLinearColor(LinearColor=&ColorFromCurve(ColorCurve=DataInterface() {propertyValues {
        AlphaCurve {keys [<at 0 1 RCIM_Cubic>]}
        BlueCurve {keys [<at 0.0011 0>, <at 0.2868 1.781>, {interpMode=RCIM_Cubic, leaveTangent=3.6939, time=0.7706, value=5}]}
        GreenCurve {keys [<at 0.0011 6.6307>, <at 0.2868 0.2704>, {interpMode=RCIM_Cubic, leaveTangent=2.3522, time=0.7706, value=0.3672}]}
        RedCurve {keys [<at 0.0011 10>, <at 0.2868 5>, {interpMode=RCIM_Cubic, leaveTangent=-4.9186, time=0.7706, value=1.2753}]}
       }}))
    ScaleRGBA=true
   ) {module=ScaleColor001}
   &ScaleSpriteSizeBySpeed(
    Max Scale Factor=<1.666 13>
    Min Scale Factor=<1.333 1.333>
    Sample Scale Factor By Curve=true
    Scale Factor Curve=DataInterface() {propertyValues {Curve {keys [<at 0 0>, {arriveTangent=1.4989, interpMode=RCIM_Cubic, tangentMode=RCTM_Break, time=1, value=1}]}}}
    Velocity Threshold=666
   )
   &ScaleSpriteSize(
    Uniform Curve Scale=&MakeFloatFromVector(Vector=$Engine.Owner.Scale, Channel=::NewEnumerator0)
    Uniform Curve Sprite Scale=DataInterface() {propertyValues {Curve {keys [
        {tangentMode=RCTM_User, time=0, value=1}
        {arriveTangent=-0.0405, interpMode=RCIM_Cubic, leaveTangent=-0.0405, tangentMode=RCTM_User, time=0.5, value=1}
        {arriveTangent=-8.0089, interpMode=RCIM_Cubic, leaveTangent=-8.0089, tangentMode=RCTM_User, time=1, value=0}
       ]}}}
   )
   &ScaleSpriteSize(
    Scale Sprite Size Mode=::NewEnumerator0
    Uniform Scale Factor=&Lerp_Float(A=0.5, B=1, Alpha=$System.Timeline.NormalizedTime)
   ) {module=ScaleSpriteSize001}
  ]
  rendererClasses [&NiagaraSpriteRendererProperties]
  renderers [{rendererClass=&NiagaraSpriteRendererProperties}]
 }
 EngineFlames {
  emitterUpdateScript [
   &EmitterState(Life Cycle Mode=::NewEnumerator1, Loop Behavior=::NewEnumerator1, Loop Duration=4.2167, UseLoopDelay=true)
   &SpawnBurst_Instantaneous(Spawn Count=1, Use Spawn Probability=true)
   &SpawnBurst_Instantaneous(Spawn Count=1, Spawn Group=1, Use Spawn Probability=true) {module=SpawnBurst_Instantaneous001}
   /SetVariables(
    Emitter.Flame1A:LinearColor=<rgba 0.468 0.1861 0.974 1>
    Emitter.Flame1B:LinearColor=<rgba 0.0414 0.0261 0.125 1>
    Emitter.Flame2A:LinearColor=<rgba 0.4483 0.4292 0.8021 1>
    Emitter.Flame2B:LinearColor=<rgba 2 0.0338 0.7315 1>
    Emitter.FlameCoreA:LinearColor=<rgba 2 0 1.3786 1>
    Emitter.FlameCoreB:LinearColor=<rgba 1 0 0.274 1>
    Emitter.FlameDark:LinearColor=<rgba 0.141 0.4896 0.3185 1>
    Emitter.FlameLight:LinearColor=<rgba 1 0.5055 0.3998 1>
   ) {isSetParameters=true}
  ]
  particleSpawnScript [
   /SetVariables(Particles.VisibilityTag=$Emitter.SpawnGroup) {isSetParameters=true}
   &InitializeParticle(
    Color=&MergeHelper_ColorSwitch(0=<rgba 1 1 1 1>, 1=<rgba 0.5 0.5 0.5 1>, 2=?, 3=?, Index=$Particles.VisibilityTag, Num=::NewEnumerator0)
    Lifetime=0.25
    Mesh Scale=&MergeHelper_VectorSwitch(0=<0.14 0.14 0.2>, 1=<0.2 0.2 0.2>, 2=?, 3=?, Index=$Particles.VisibilityTag, Num=::NewEnumerator0)
    Mesh Scale Mode=::NewEnumerator4
    UsePositionOffset=false
    Write Material Random=true
   )
   /SetVariables(
    Particles.Flame2Scale:Vector3f=&Multiply_Vector(A=$Particles.Scale, B=<1 1 0.9>)
    Particles.FlameCoreScale:Vector3f=&Multiply_Vector(A=$Particles.Scale, B=<0.2 0.2 0.8>)
   ) {isSetParameters=true}
  ]
  particleUpdateScript [
   &ParticleState(Kill Particles When Lifetime Has Elapsed=false)
   /SetVariables(Particles.RandomScaling:Vector3f=&Multiply_Vector(
     A=&MakeVector(
      X=&Waveform("Waveform [1]"=::NewEnumerator0, "[1] Frequency"=4, "[1] Amplitude Min/Max"=<0.8 1>)
      Y=&Waveform("Waveform [1]"=::NewEnumerator0, "[1] Frequency"=3, "[1] Amplitude Min/Max"=<0.8 1>)
      Z=&Waveform(
       Waveform Count=::NewEnumerator1
       "Waveform [1]"=::NewEnumerator9
       "[1] Frequency"=14
       "[1] Amplitude Min/Max"=<0.8 1.1>
       "Waveform [2] Blend"=::NewEnumerator2
       "Waveform [2]"=::NewEnumerator6
       "[2] Input Position"=0
       "[2] Phase"=$Emitter.Age
       "[2] Phase Scale"=1
       "[2] Compound Frequency"=2
       "[2] Frequency"=0.34
       "[2] Amplitude Min/Max"=<0.9 1.3>
       "[2] Scale"=1
       "[2] Bias"=0
       "[2] Exponent"=1
       "[2] Pulse Duration"=0.1
       "[2] Clip Triangle Min/Max"=<0 0>
       "[2] Random Seed"=0
      )
     )
     B=&Multiply_VectorByFloat(Vector=<1.1 1.1 1>, Float=1)
    )) {isSetParameters=true}
   &SocketLocation(SamplePosOffset=false, Socket Index=2, Socket Reader=$System.Socket Reader, Socket Sampling Mode=::NewEnumerator3, VelocitySpeedLimit=false)
   &SolveForcesAndVelocity(Clamp Velocity=false, Limit Acceleration=false)
   &ScaleColor(
    Scale Alpha=&FloatFromCurve(
     FloatCurve=DataInterface() {propertyValues {Curve {keys [<at 0 0 RCIM_Cubic>,<at 0.5 1 RCIM_Cubic>]}}}
     Scale Curve=&MergeHelper_FloatSwitch(
      0=0.3
      1=&RemapRange(Input Value To Remap=$System.CameraDot, Input Min=0, Input Max=1, Output Min=0, Output Max=0.8)
      2=?
      3=?
      Index=$Particles.VisibilityTag
      Num=::NewEnumerator0
     )
    )
    Scale RGB=&MergeHelper_VectorSwitch(0=<1 1 1>, 1=<0.2 0.2 0.2>, 2=?, 3=?, Index=$Particles.VisibilityTag, Num=::NewEnumerator0)
    ScaleA=true
    ScaleRGB=true
   )
   &ScaleMeshSize(Scale Factor=&Multiply_Vector(
     A=&MakeVector(X=1, Y=1, Z=$System.AlphaFade)
     B=&MergeHelper_VectorSwitch(0=$Particles.RandomScaling, 1=<1 1 1>, 2=?, 3=?, Index=$Particles.VisibilityTag, Num=::NewEnumerator0)
    ))
   /SetVariables(
    Particles.Flame2Scale:Vector3f=&Multiply_Vector(
     A=&Multiply_Vector(A=$Particles.Initial.Flame2Scale, B=$Particles.RandomScaling)
     B=&MakeVector(X=1, Y=1, Z=$System.AlphaFade)
    )
    Particles.FlameCoreScale:Vector3f=&Multiply_Vector(
     A=&Multiply_Vector(A=$Particles.Initial.FlameCoreScale, B=$Particles.RandomScaling)
     B=&MakeVector(X=1, Y=1, Z=$System.AlphaFade)
    )
   ) {isSetParameters=true}
   &DynamicMaterialParameters(
    Index 0 Param 1=&Lerp_Float(A=0.3, B=0.6, Alpha=1)
    Index 0 Param 2=2
    Module.Param0WriteEnabled=true
    Module.Param1WriteEnabled=true
    Module.Param2WriteEnabled=true
    Module.Param3WriteEnabled=true
   )
  ]
  rendererClasses [&NiagaraMeshRendererProperties]
  renderers [
   {rendererClass=&NiagaraMeshRendererProperties, rendererIndex=0}
   {rendererClass=&NiagaraMeshRendererProperties, rendererIndex=1}
   {rendererClass=&NiagaraMeshRendererProperties, rendererIndex=2}
   {rendererClass=&NiagaraMeshRendererProperties, rendererIndex=3}
   {rendererClass=&NiagaraMeshRendererProperties, rendererIndex=4}
  ]
 }
 ExhaustGlow {
  emitterUpdateScript [
   &EmitterState(
    Inactive Response=::NewEnumerator1
    Life Cycle Mode=::NewEnumerator1
    Loop Behavior=::NewEnumerator1
    Loop Duration Mode=::NewEnumerator1
    UseLoopDelay=false
   )
   &SpawnBurst_Instantaneous(Use Spawn Probability=false, UseLoopCountLimit=false)
   &SpawnRate(SpawnGroup=1, SpawnRate=4, Use Spawn Probability=false)
   &SpawnRate(SpawnGroup=2, SpawnRate=3, Use Spawn Probability=true) {module=SpawnRate001}
  ]
  particleSpawnScript [
   /SetVariables(Particles.VisibilityTag=$Emitter.SpawnGroup) {isSetParameters=true}
   &InitializeParticle(
    Color=&MergeHelper_ColorSwitch(0=<rgba 0.2784 0.2282 2 0.75>, 1=<rgba 3 0 2.9959 1>, 2=<rgba 4.2395 0 5 0.5>, 3=?, Index=$Particles.VisibilityTag, Num=::NewEnumerator1)
    Lifetime=&MergeHelper_FloatSwitch(
     0=1
     1=&RandomRangeFloat(Minimum=0.5, Maximum=0.8)
     2=&RandomRangeFloat(Minimum=0.5, Maximum=0.7)
     3=?
     Index=$Particles.VisibilityTag
     Num=::NewEnumerator1
    )
    Sprite Rotation Angle=&MergeHelper_FloatSwitch(
     0=0
     1=&RandomRangeFloat(Minimum=0, Maximum=1)
     2=0
     3=?
     Index=$Particles.VisibilityTag
     Num=::NewEnumerator1
    )
    Sprite Rotation Mode=::NewEnumerator1
    Sprite Size=&MergeHelper_Vector2Switch(
     0=<12 20>
     1=&RandomRangeVector2D(Minimum=<40 30>, Maximum=<30 50>)
     2=&RandomRangeVector2D(Minimum=<30 60>, Maximum=<40 50>)
     3=?
     4=?
     Index=$Particles.VisibilityTag
     Num=::NewEnumerator1
    )
    Sprite Size Mode=::NewEnumerator4
    UsePositionOffset=true
    Write Material Random=true
   )
  ]
  particleUpdateScript [
   &ParticleState(Kill Particles When Lifetime Has Elapsed=false)
   &SolveForcesAndVelocity(Clamp Velocity=false, Limit Acceleration=false)
   /SetVariables(Particles.Jitter:NiagaraFloat=&Waveform(
     Global Input Position=&RandomRangeFloat(Minimum=0, Maximum=1)
     "Waveform [1]"=::NewEnumerator12
     "[1] Frequency"=15
     "[1] Amplitude Min/Max"=<0.66 1>
    )) {isSetParameters=true}
   &SocketLocation(SamplePosOffset=false, Socket Index=2, Socket Reader=$System.Socket Reader, Socket Sampling Mode=::NewEnumerator3, VelocitySpeedLimit=false)
   &ScaleSpriteSize(
    Scale Factor=&Multiply_Vector2D_ByFloat(
     Float=$System.AlphaFade
     Vector2D=&MergeHelper_Vector2Switch(
      0=&MakeVector2D(
       X=&Waveform(
        Waveform Count=::NewEnumerator1
        "Waveform [1]"=::NewEnumerator9
        "[1] Frequency"=3
        "[1] Amplitude Min/Max"=<1 0.8>
        "Waveform [2] Blend"=::NewEnumerator2
        "[2] Input Position"=0
        "[2] Phase"=$Emitter.Age
        "[2] Phase Scale"=1
        "[2] Compound Frequency"=2
        "[2] Frequency"=1.4
        "[2] Amplitude Min/Max"=<0.9 1>
        "[2] Scale"=1
        "[2] Bias"=0
        "[2] Exponent"=1
        "[2] Pulse Duration"=0.1
        "[2] Clip Triangle Min/Max"=<0 0>
        "[2] Random Seed"=0
       )
       Y=&Waveform(
        Waveform Count=::NewEnumerator1
        "Waveform [1]"=::NewEnumerator9
        "[1] Frequency"=3
        "[1] Amplitude Min/Max"=<0.8 2>
        "Waveform [2] Blend"=::NewEnumerator2
        "[2] Input Position"=0
        "[2] Phase"=$Emitter.Age
        "[2] Phase Scale"=1
        "[2] Compound Frequency"=2
        "[2] Frequency"=1.4
        "[2] Amplitude Min/Max"=<0.9 1>
        "[2] Scale"=1
        "[2] Bias"=0
        "[2] Exponent"=1
        "[2] Pulse Duration"=0.1
        "[2] Clip Triangle Min/Max"=<0 0>
        "[2] Random Seed"=0
       )
      )
      1=&Vector2DFromFloat(Value=&Multiply_Float(
        A=&FloatFromCurve(
         FloatCurve=DataInterface() {propertyValues {Curve {keys [{arriveTangent=2.1981, interpMode=RCIM_Cubic, leaveTangent=2.1981, tangentMode=RCTM_User, time=0, value=0}, <at 1 1 RCIM_Cubic>]}}}
         Scale Curve=$System.CameraDot
        )
        B=&Waveform(
         Waveform Count=::NewEnumerator1
         "Waveform [1]"=::NewEnumerator9
         "[1] Frequency"=3
         "[1] Amplitude Min/Max"=<0.8 1.1>
         "Waveform [2] Blend"=::NewEnumerator2
         "[2] Input Position"=0
         "[2] Phase"=$Emitter.Age
         "[2] Phase Scale"=1
         "[2] Compound Frequency"=2
         "[2] Frequency"=1.4
         "[2] Amplitude Min/Max"=<0.9 1>
         "[2] Scale"=1
         "[2] Bias"=0
         "[2] Exponent"=1
         "[2] Pulse Duration"=0.1
         "[2] Clip Triangle Min/Max"=<0 0>
         "[2] Random Seed"=0
        )
       ))
      2=&Vector2DFromFloat(Value=&Multiply_Float(
        A=&FloatFromCurve(
         FloatCurve=DataInterface() {propertyValues {Curve {keys [{arriveTangent=2.1981, interpMode=RCIM_Cubic, leaveTangent=2.1981, tangentMode=RCTM_User, time=0, value=0}, <at 1 1 RCIM_Cubic>]}}}
         Scale Curve=$System.CameraDot
        )
        B=&Waveform(
         Waveform Count=::NewEnumerator1
         "Waveform [1]"=::NewEnumerator9
         "[1] Frequency"=3
         "[1] Amplitude Min/Max"=<0.8 1.1>
         "Waveform [2] Blend"=::NewEnumerator2
         "[2] Input Position"=0
         "[2] Phase"=$Emitter.Age
         "[2] Phase Scale"=1
         "[2] Compound Frequency"=2
         "[2] Frequency"=1.4
         "[2] Amplitude Min/Max"=<0.9 1>
         "[2] Scale"=1
         "[2] Bias"=0
         "[2] Exponent"=1
         "[2] Pulse Duration"=0.1
         "[2] Clip Triangle Min/Max"=<0 0>
         "[2] Random Seed"=0
        )
       ))
      3=?
      4=?
      Index=$Particles.VisibilityTag
      Num=::NewEnumerator1
     )
    )
    Scale Sprite Size Mode=::NewEnumerator2
   )
   &ScaleColor(
    Scale Alpha=&MergeHelper_FloatSwitch(
     0=&Waveform(
      Waveform Count=::NewEnumerator1
      "Waveform [1]"=::NewEnumerator9
      "[1] Frequency"=20
      "[1] Amplitude Min/Max"=<1 2>
      "Waveform [2] Blend"=::NewEnumerator2
      "[2] Input Position"=0
      "[2] Phase"=$Emitter.Age
      "[2] Phase Scale"=1
      "[2] Compound Frequency"=2
      "[2] Frequency"=1.4
      "[2] Amplitude Min/Max"=<0.5 1>
      "[2] Scale"=1
      "[2] Bias"=0
      "[2] Exponent"=1
      "[2] Pulse Duration"=0.1
      "[2] Clip Triangle Min/Max"=<0 0>
      "[2] Random Seed"=0
     )
     1=&FloatFromCurve(
      FloatCurve=DataInterface() {propertyValues {Curve {keys [<at 0 0 RCIM_Cubic>,<at 0.2 1 RCIM_Cubic>,<at 1 0 RCIM_Cubic>]}}}
      Scale Curve=$System.CameraDot
     )
     2=&FloatFromCurve(
      FloatCurve=DataInterface() {propertyValues {Curve {keys [<at 0 0 RCIM_Cubic>,<at 0.2 1 RCIM_Cubic>,<at 1 0 RCIM_Cubic>]}}}
      Scale Curve=$System.CameraDot
     )
     3=?
     Index=$Particles.VisibilityTag
     Num=::NewEnumerator1
    )
    ScaleA=true
    ScaleRGB=true
   )
   &FN_Light_Attributes(
    Color=<rgba 0.4178 0 1 1>
    Create Color Attribute=true
    Create Exponent Attribute=true
    Create Position Attribute=true
    Create Radius Attribute=true
    Exponent=5
    Position=&AddVectorToPosition(
     Position=$Output.SocketLocation.SampledPosition
     Vector=&TransformVector(Vector=<0 0 10>, Destination Space=::Simulation)
    )
    Radius=30
    Scale Color=&Multiply_Float(
     A=&Waveform(
      Waveform Count=::NewEnumerator1
      "Waveform [1]"=::NewEnumerator9
      "[1] Frequency"=3.2
      "[1] Amplitude Min/Max"=<10 100>
      "Waveform [2] Blend"=::NewEnumerator2
      "[2] Input Position"=0
      "[2] Phase"=$Emitter.Age
      "[2] Phase Scale"=1
      "[2] Compound Frequency"=2
      "[2] Frequency"=7
      "[2] Amplitude Min/Max"=<0.8 1>
      "[2] Scale"=1
      "[2] Bias"=0
      "[2] Exponent"=1
      "[2] Pulse Duration"=0.1
      "[2] Clip Triangle Min/Max"=<0 0>
      "[2] Random Seed"=0
     )
     B=$System.AlphaFade
    )
   )
   &CameraOffset(Camera Offset Amount=&MergeHelper_FloatSwitch(
     0=5
     1=&Lerp_Float(
      A=0
      B=20
      Alpha=&RemapRange(Input Value To Remap=$System.CameraDot, Input Min=0.8, Input Max=1, Output Min=0, Output Max=1)
     )
     2=&Lerp_Float(
      A=0
      B=20
      Alpha=&RemapRange(Input Value To Remap=$System.CameraDot, Input Min=0.8, Input Max=1, Output Min=0, Output Max=1)
     )
     3=?
     Index=$Particles.VisibilityTag
     Num=::NewEnumerator1
    )) {module=CameraOffset001}
   &SpriteFacingAndAlignment(
    Alignment Coordinate Space=::Local
    Sprite Alignment=&MergeHelper_VectorSwitch(0=<0 0 1>, 1=<1 0 0>, 2=<1 0 0>, 3=?, Index=$Particles.VisibilityTag, Num=::NewEnumerator1)
    WriteAlignment=true
    WriteFacing=false
   )
   &KillParticles(Kill Particles=&SetBoolByFloatComparison(
     A=&MergeHelper_FloatSwitch(0=0, 1=$Particles.NormalizedAge, 2=$Particles.NormalizedAge, 3=?, Index=$Particles.VisibilityTag, Num=::NewEnumerator1)
     B=1
    ))
  ]
  rendererClasses [&NiagaraSpriteRendererProperties, &NiagaraLightRendererProperties]
  renderers [
   {rendererClass=&NiagaraSpriteRendererProperties}
   {rendererClass=&NiagaraLightRendererProperties, rendererIndex=1}
   {rendererClass=&NiagaraSpriteRendererProperties, rendererIndex=2}
   {rendererClass=&NiagaraSpriteRendererProperties, rendererIndex=3}
  ]
 }
 Flare {
  emitterUpdateScript [
   &EmitterState(Life Cycle Mode=::NewEnumerator1, Loop Behavior=::NewEnumerator1, Loop Delay=0.5, Loop Duration=0.2, UseLoopDelay=true)
   &SpawnBurst_Instantaneous(Spawn Count=1, Use Spawn Probability=false)
   &SpawnBurst_Instantaneous(Spawn Count=1, Spawn Group=1, Use Spawn Probability=false) {module=SpawnBurst_Instantaneous001}
  ]
  particleSpawnScript [
   /SetVariables(Particles.VisibilityTag=&MergeHelper_IntSwitch(0=0, 1=1, 2=2, 3=2, 4=?, 5=?, Index=$Emitter.SpawnGroup, Num=::NewEnumerator2)) {isSetParameters=true}
   /SetVariables(Particles.CustomIndex:NiagaraInt32=$Emitter.SpawnGroup) {isSetParameters=true}
   &InitializeParticle(
    Color=<rgba 5 5 5 0.1>
    Lifetime=0.3
    Sprite Rotation Mode=::NewEnumerator2
    Sprite Size Mode=::NewEnumerator3
    Uniform Sprite Size=20
    UsePositionOffset=false
    Write Material Random=true
   )
   &DynamicMaterialParameters(
    Index 0 Param 1=&RandomRangeFloat(Minimum=0.4, Maximum=0.6)
    Index 0 Param 2=1
    Index 0 Param 3=1
    Index 0 Param 4=1
    Module.Index0Float0Write=true
    Module.Index0Float1Write=true
    Module.Index0Float2Write=true
    Module.Index0Float3Write=true
   )
   &InitialMeshOrientation(Orientation Axis=<0 0 1>, Orientation Vector=<0 0 1>, Rotation=<0 0.25 0>, Use Orientation Vector=true, Use Rotation Vector=true)
   &CameraOffset(Camera Offset Amount=&MakeCustomFloatFromBool(Bool=$User.bIsFrontEnd, True Float=10, False Float=10))
  ]
  particleUpdateScript [
   /SetVariables(Particles.Dark_Glow=&MakeLinearColorFromVectorAndFloat(
     "Vector (RGB)"=&MakeVectorFromLinearColorRGB(Color=&MultiplyLinearColor(
       Linear Color=&ColorFromCurve(ColorCurve=DataInterface() {propertyValues {
          AlphaCurve {keys []}
          BlueCurve {keys [<at 0.2868 0.3218 RCIM_Cubic>,<at 0.5555 0.6354>,<at 0.9372 0.5784 RCIM_Cubic>]}
          GreenCurve {keys [<at 0.2868 0 RCIM_Cubic>,<at 0.5555 0>,<at 0.9372 0 RCIM_Cubic>]}
          RedCurve {keys [<at 0.2868 0.3281 RCIM_Cubic>,<at 0.5555 0.3654>,<at 0.9372 0.8594 RCIM_Cubic>]}
         }})
       Scale Factor=&MakeVector4_FromFloat(Float=0.5)
      ))
     "Float (Alpha)"=&Multiply_Float(
      A=&Multiply_Float(A=0.2, B=1)
      B=&FloatFromCurve(FloatCurve=DataInterface() {propertyValues {Curve {keys [<at 0 0>,<at 0.2 1>,<at 0.7 0.4>,<at 1 0>]}}})
     )
    )) {isSetParameters=true}
   /SetVariables(Particles.Dark_Glow_Size=&Multiply_Vector2(
     A=&Vector2DFromFloat(Value=30)
     B=&Vector2DFromFloat(Value=&FloatFromCurve(
       FloatCurve=DataInterface() {propertyValues {Curve {keys [<at 0 1.5 RCIM_Cubic>,<at 1 0.5>]}}}
       Scale Curve=&FN_FrontEndFloatModifier(Gameplay=1.5, Front End=0.6)
      ))
    )) {isSetParameters=true}
   &SocketLocation(
    SamplePosOffset=true
    Sampled Position Offset=<0 0 2>
    Socket Index=2
    Socket Reader=$System.Socket Reader
    Socket Sampling Mode=::NewEnumerator3
    VelocitySpeedLimit=false
   )
   &ParticleState()
   &ScaleMeshSize(Scale Factor=&VectorFromFloat(Value=&FloatFromCurve(
      FloatCurve=DataInterface() {propertyValues {Curve {keys [{arriveTangent=4.1564, interpMode=RCIM_Cubic, leaveTangent=4.1564, tangentMode=RCTM_User, time=0, value=0}, <at 1 1.4 RCIM_Cubic>]}}}
      Scale Curve=&Multiply_Float(
       A=&Multiply_Float(
        A=&MergeHelper_FloatSwitch(0=0, 1=0, 2=0.9, 3=1.3, Index=$Particles.CustomIndex, Num=::NewEnumerator2)
        B=&FN_FrontEndFloatModifier(Gameplay=1, Front End=0.6)
       )
       B=$System.Scale
      )
     )))
   &ScaleColor(
    Scale Alpha=&FloatFromCurve(FloatCurve=DataInterface() {propertyValues {Curve {keys [<at 0 0>,<at 0.05 1>,<at 0.6 0.6>,<at 0.7 -0.05>]}}})
    Scale RGB=&MakeVectorFromLinearColorRGB(Color=&ColorFromCurve(ColorCurve=DataInterface() {propertyValues {
        AlphaCurve {keys [<at 0 1>,<at 1 0>]}
        BlueCurve {keys [<at 0 0>,<at 0.2484 2.6124>,<at 1 1>]}
        GreenCurve {keys [<at 0 2.3087>,<at 0.2484 0>,<at 1 0>]}
        RedCurve {keys [<at 0 5>,<at 0.2484 5>,<at 1 0.457>]}
       }}))
    ScaleA=true
    ScaleRGB=true
   )
   &Color(
    Color=&MergeHelper_ColorSwitch(
     0=$Particles.Color
     1=$Particles.Color
     2=&ColorFromCurve(ColorCurve=DataInterface() {propertyValues {
        AlphaCurve {keys [<at 0 0.5>,<at 1 0>]}
        BlueCurve {keys [<at 0 0.1937>,<at 0.1684 1>,<at 0.864 0.1456>]}
        GreenCurve {keys [<at 0 0.0595>,<at 0.1684 0>,<at 0.864 0.0282>]}
        RedCurve {keys [<at 0 1>,<at 0.1684 0.8167>,<at 0.864 0.2>]}
       }})
     3=&MultiplyLinearColor(
      Linear Color=<rgba 0.0393 0.0355 0.0885 1>
      Scale Factor=&MakeVector4_FromLinearColor(LinearColor=&ColorFromCurve(ColorCurve=DataInterface() {propertyValues {
          AlphaCurve {keys [<at 0 0.5>,<at 1 0>]}
          BlueCurve {keys [<at 0 1>,<at 0.1684 1>,<at 0.4403 0.9636>]}
          GreenCurve {keys [<at 0 0>,<at 0.1684 0>,<at 0.4403 0.272>]}
          RedCurve {keys [<at 0 0.6167>,<at 0.1684 0.9667>,<at 0.4403 1>]}
         }}))
     )
     Index=$Particles.CustomIndex
     Num=::NewEnumerator2
    )
    Scale Alpha=&MergeHelper_FloatSwitch(0=1, 1=1, 2=4, 3=4, Index=$Particles.CustomIndex, Num=::NewEnumerator2)
    Scale Color=&MergeHelper_VectorSwitch(
     0=<1 1 1>
     1=<1 1 1>
     2=&VectorFromFloat(Value=&FloatFromCurve(
       FloatCurve=DataInterface() {propertyValues {Curve {keys [<at 0 10>,<at 0.15 3>,<at 1 1>]}}}
       Scale Curve=&FNTimeOfDayScalar(Night Scale=0.3, Morning Scale=0.5, Day Scale=0.5, Evening Scale=0.5)
      ))
     3=&VectorFromFloat(Value=&FloatFromCurve(
       FloatCurve=DataInterface() {propertyValues {Curve {keys [<at 0 10>,<at 0.15 3>,<at 1 1>]}}}
       Scale Curve=&FNTimeOfDayScalar(Night Scale=0.3, Morning Scale=0.5, Day Scale=0.5, Evening Scale=0.5)
      ))
     Index=$Particles.CustomIndex
     Num=::NewEnumerator2
    )
   )
   &ScaleSpriteSize(Scale Factor=&Vector2DFromFloat(Value=&MergeHelper_FloatSwitch(
      0=&FloatFromCurve(
       FloatCurve=DataInterface() {propertyValues {Curve {keys [<at 0 1.5 RCIM_Cubic>,<at 0.8 0>]}}}
       Scale Curve=&FN_FrontEndFloatModifier(Gameplay=1.3, Front End=1)
      )
      1=&FloatFromCurve(
       FloatCurve=DataInterface() {propertyValues {Curve {keys [{arriveTangent=2.9429, interpMode=RCIM_Cubic, leaveTangent=2.9429, tangentMode=RCTM_User, time=0, value=0}, <at 1 1 RCIM_Cubic>]}}}
       Scale Curve=&FN_FrontEndFloatModifier(Gameplay=1.5, Front End=1)
      )
      2=?
      3=?
      Index=$Particles.VisibilityTag
      Num=::NewEnumerator0
     )))
   /SetVariables(Particles.DynamicMatMesh:Vector4f=&MakeVector4(
     X=&FloatFromCurve(FloatCurve=DataInterface() {propertyValues {Curve {keys [{arriveTangent=-2.1015, interpMode=RCIM_Cubic, leaveTangent=-2.1015, tangentMode=RCTM_User, time=0, value=0.5}, <at 1 -0.15>]}}})
     Y=&FloatFromCurve(FloatCurve=DataInterface() {propertyValues {Curve {keys [{arriveTangent=-0.3464, interpMode=RCIM_Cubic, leaveTangent=-0.3464, tangentMode=RCTM_User, time=0, value=1}, <at 1 0.9>]}}})
     Z=&FloatFromCurve(FloatCurve=DataInterface() {propertyValues {Curve {keys [<at 0 0>,<at 0.15 1>,<at 1 1.2>]}}})
     W=0
    )) {isSetParameters=true}
   &FN_Light_Attributes(
    Color=<rgba 1 0 0.9625 1>
    Create Color Attribute=true
    Create Exponent Attribute=true
    Create Position Attribute=true
    Create Radius Attribute=true
    Exponent=5
    Position=&AddVectorToPosition(
     Position=$Output.SocketLocation.SampledPosition
     Vector=&TransformVector(Vector=<0 0 10>, Destination Space=::Simulation)
    )
    Radius=30
    Scale Color=&FloatFromCurve(FloatCurve=DataInterface() {propertyValues {Curve {keys [<at 0 0>,<at 0.2 1>,<at 0.5 0>]}}}, Scale Curve=200)
   )
  ]
  rendererClasses [&NiagaraSpriteRendererProperties, &NiagaraLightRendererProperties]
  renderers [
   {rendererClass=&NiagaraSpriteRendererProperties}
   {rendererClass=&NiagaraLightRendererProperties, rendererIndex=1}
   {rendererClass=&NiagaraSpriteRendererProperties, rendererIndex=2}
  ]
 }
}
summary {
 systemName=NS_Pickaxe_RenegadeWhip_Idle
 emitters [
  {emitterName=ExhaustGlow, rendererClasses [&NiagaraSpriteRendererProperties, &NiagaraLightRendererProperties]}
  {emitterName=Embers, simTarget=GPUComputeSim, rendererClasses [&NiagaraSpriteRendererProperties]}
  {emitterName=EngineFlames, rendererClasses [&NiagaraMeshRendererProperties]}
  {emitterName=Flare, rendererClasses [&NiagaraSpriteRendererProperties, &NiagaraLightRendererProperties]}
 ]
 userVariables [
  User.bIsFrontEnd {type=NiagaraBool}
  User.bIsFrontEndPreview {type=NiagaraBool}
  User.bIsFullyDeployed {type=NiagaraBool}
  User.bIsSwinging {type=NiagaraBool}
  User.ForwardDot {type=NiagaraFloat, defaultValue=0, description=""}
  User.Variant {type=NiagaraFloat, defaultValue=1, description=""}
 ]
}
