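# Animation Blueprint for the RenegadeWhip pickaxe that adds secondary motion to its broom and charm bones.
# Two AnimDynamics chains start from a cached input pose: world-space with an additive handle rotation, and handle-relative with mostly looser acceleration clamps.
# A Blend Poses by bool feeds the output, with no driver shown.
# The update event sets GravityOverride through a deprecated function, but no graph node reads it.
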AnimBlueprint /BRCosmetics/Weapons/FORT_Melee/Pickaxe_RenegadeWhip/Meshes/RenegadeWhip_Axe_AnimBlueprint
target_skeleton=&/RenegadeWhip_Axe_Skeleton
graphs [
 AnimGraph {nodes [
   Anim Dynamics - Bone: dyn_broom_01(Alpha=0.3, ExternalForce="0,0,0") {
    angularBiasOverride=0.5
    angularDampingOverride=0.9999
    angularSpringConstant=50
    bAngularSpring=true
    simulationSpace=World
    alphaBoolBlend {blendOption=Linear}
    boundBone {boneName=dyn_broom_01}
    componentAppliedLinearAccClamp=<100 100 100>
    gravityOverride {z=-980}
    physicsBodyDefinitions [{
      boundBone {boneName=dyn_broom_01}
      boxExtents=<40 40 40>
      constraintSetup {angularLimitsMax=<0 5 5>, angularLimitsMin=<0 -5 -5>, angularTarget=<1 0 0>}
      localJointOffset=<0.01 0 0>
     }]
   }
   n1 = "Transform (Modify) Bone - Bone: Melee_TwoHanded_Handle"(ComponentPose=Local To Component(LocalPose=Use cached pose 'InitPose'()), Alpha=1, Translation="0,0,0", Rotation="0,-45,0", Scale="1,1,1") {rotationMode=BMM_Additive, alphaBoolBlend {blendOption=Linear}, boneToModify {boneName=Melee_TwoHanded_Handle}, rotation {yaw=-45}}
   n2 = Anim Dynamics - Bone: dyn_broom_01(ComponentPose=@n1, Alpha=1, ExternalForce="0,0,0") {
    angularDampingOverride=1
    angularSpringConstant=10000
    bAngularSpring=true
    bOverrideAngularDamping=true
    bUseGravityOverride=true
    simulationSpace=World
    alphaBoolBlend {blendOption=Linear}
    boundBone {boneName=dyn_broom_01}
    componentAppliedLinearAccClamp=<500 500 500>
    componentLinearAccScale=<0.1 0.1 0.1>
    componentLinearVelScale=<0.1 0.1 0.1>
    gravityOverride {z=-980}
    physicsBodyDefinitions [{
      boundBone {boneName=dyn_broom_01}
      boxExtents=<30 30 30>
      constraintSetup {angularLimitsMax=<0 7 7>, angularLimitsMin=<0 -7 -7>, angularTarget=<1 0 0>}
      localJointOffset=<0 0 0>
     }]
   }
   n3 = Anim Dynamics - Bone: dyn_charm_01(ComponentPose=@n2, Alpha=1, ExternalForce="0,0,0") {
    angularBiasOverride=0.5
    angularDampingOverride=0.999
    angularSpringConstant=100
    bAngularSpring=true
    bOverrideAngularBias=true
    bOverrideAngularDamping=true
    bUseGravityOverride=true
    bUsePlanarLimit=false
    simulationSpace=World
    alphaBoolBlend {blendOption=Linear}
    boundBone {boneName=dyn_charm_01}
    componentAppliedLinearAccClamp=<1 1 1>
    gravityOverride {z=-980}
    physicsBodyDefinitions [{
      boundBone {boneName=dyn_charm_01}
      boxExtents=<10 10 10>
      constraintSetup {angularLimitsMax=<5 45 45>, angularLimitsMin=<-5 -45 -45>, angularTarget=<1 0 0>}
      localJointOffset=<0.54 0 0>
     }]
    planarLimits [{drivingBone {boneName=Melee_TwoHanded_Handle}, planeTransform {location=<0 0 -35>, rotation=<rot 0 90 90>}}]
    relativeSpaceBone {boneName=Melee_TwoHanded_Handle}
    sphericalLimits [{limitRadius=0, drivingBone {boneName=Melee_TwoHanded_Handle}, sphereLocalOffset=<0 0 0>}]
   }
   n4 = Anim Dynamics - Bone: dyn_charm_02(ComponentPose=@n3, Alpha=1, ExternalForce="0,0,0") {
    angularBiasOverride=0.5
    angularDampingOverride=0.99
    angularSpringConstant=75
    bAngularSpring=true
    bOverrideAngularBias=true
    bOverrideAngularDamping=true
    bUseGravityOverride=true
    simulationSpace=World
    alphaBoolBlend {blendOption=Linear}
    boundBone {boneName=dyn_charm_02}
    componentAppliedLinearAccClamp=<1 1 1>
    gravityOverride {z=-980}
    physicsBodyDefinitions [{
      boundBone {boneName=dyn_charm_02}
      boxExtents=<10 10 10>
      constraintSetup {angularLimitsMax=<5 25 25>, angularLimitsMin=<-5 -25 -25>, angularTarget=<1 4 0>, linearAxesMax=<1 0 0>}
      localJointOffset=<0.54 0 0>
     }]
    planarLimits [{drivingBone {boneName=Melee_TwoHanded_Handle}, planeTransform {location=<5 0 -35>, rotation=<rot 0 90 90>}}]
    relativeSpaceBone {boneName=Melee_TwoHanded_Handle}
    sphericalLimits [{limitRadius=0, drivingBone {boneName=Melee_TwoHanded_Handle}, sphereLocalOffset=<0 0 0>}]
   }
   n5 = Anim Dynamics - Bone: dyn_charm_03(ComponentPose=@n4, Alpha=1, ExternalForce="0,0,0") {
    angularBiasOverride=0.5
    angularDampingOverride=0.999
    angularSpringConstant=75
    bAngularSpring=true
    bOverrideAngularBias=true
    bOverrideAngularDamping=true
    bUseGravityOverride=true
    bUsePlanarLimit=false
    simulationSpace=World
    alphaBoolBlend {blendOption=Linear}
    boundBone {boneName=dyn_charm_03}
    componentAppliedLinearAccClamp=<1 1 1>
    gravityOverride {z=-980}
    physicsBodyDefinitions [{
      sphereCollisionRadius=7
      boundBone {boneName=dyn_charm_03}
      boxExtents=<10 10 5>
      constraintSetup {angularLimitsMax=<5 25 25>, angularLimitsMin=<-5 -25 -25>, angularTarget=<1 0 0>}
      localJointOffset=<1 0 0>
     }]
    planarLimits [{drivingBone {boneName=Melee_TwoHanded_Handle}, planeTransform {location=<-3 0 -35>, rotation=<rot 0 90 90>}}]
    relativeSpaceBone {boneName=Melee_TwoHanded_Handle}
    sphericalLimits [{limitRadius=100, limitType=Inner, drivingBone {boneName=Melee_TwoHanded_Handle}, sphereLocalOffset=<-100 10 -35>}]
   }
   n6 = Anim Dynamics - Bone: dyn_broom_01(ComponentPose=Local To Component(LocalPose=Use cached pose 'InitPose'()), Alpha=1, ExternalForce="0,0,0") {
    angularDampingOverride=1
    angularSpringConstant=5000
    bAngularSpring=true
    bOverrideAngularDamping=true
    bUseGravityOverride=true
    simulationSpace=BoneRelative
    alphaBoolBlend {blendOption=Linear}
    boundBone {boneName=dyn_broom_01}
    componentAppliedLinearAccClamp=<3000 3000 3000>
    componentLinearAccScale=<0.3 0.3 0.3>
    componentLinearVelScale=<0.3 0.3 0.3>
    gravityOverride {z=-980}
    physicsBodyDefinitions [{
      boundBone {boneName=dyn_broom_01}
      boxExtents=<15 15 15>
      constraintSetup {angularLimitsMax=<30 30 30>, angularLimitsMin=<-30 -30 -30>, angularTarget=<1 0 0>}
      localJointOffset=<1 0 0>
     }]
    relativeSpaceBone {boneName=Melee_TwoHanded_Handle}
   }
   n7 = Anim Dynamics - Bone: dyn_charm_01(ComponentPose=@n6, Alpha=1, ExternalForce="0,0,0") {
    angularBiasOverride=0.5
    angularDampingOverride=0.999
    angularSpringConstant=100
    bAngularSpring=true
    bOverrideAngularBias=true
    bOverrideAngularDamping=true
    bUseGravityOverride=true
    bUsePlanarLimit=false
    simulationSpace=BoneRelative
    alphaBoolBlend {blendOption=Linear}
    boundBone {boneName=dyn_charm_01}
    componentAppliedLinearAccClamp=<4000 4000 4000>
    componentLinearAccScale=<1 2 2>
    componentLinearVelScale=<1 2 2>
    gravityOverride {z=-980}
    physicsBodyDefinitions [{
      boundBone {boneName=dyn_charm_01}
      boxExtents=<10 10 10>
      constraintSetup {angularLimitsMax=<5 25 25>, angularLimitsMin=<-5 -25 -25>, angularTarget=<1 0 0>}
      localJointOffset=<0.54 0 0>
     }]
    planarLimits [{drivingBone {boneName=Melee_TwoHanded_Handle}, planeTransform {location=<0 0 -35>, rotation=<rot 0 90 90>}}]
    relativeSpaceBone {boneName=Melee_TwoHanded_Handle}
    sphericalLimits [{limitRadius=0, drivingBone {boneName=Melee_TwoHanded_Handle}, sphereLocalOffset=<0 0 0>}]
   }
   n8 = Anim Dynamics - Bone: dyn_charm_02(ComponentPose=@n7, Alpha=1, ExternalForce="0,0,0") {
    angularBiasOverride=0.5
    angularDampingOverride=0.99
    angularSpringConstant=75
    bAngularSpring=true
    bOverrideAngularBias=true
    bOverrideAngularDamping=true
    bUseGravityOverride=true
    simulationSpace=BoneRelative
    alphaBoolBlend {blendOption=Linear}
    boundBone {boneName=dyn_charm_02}
    componentAppliedLinearAccClamp=<4000 4000 4000>
    componentLinearAccScale=<1 2 2>
    componentLinearVelScale=<1 2 2>
    gravityOverride {z=-980}
    physicsBodyDefinitions [{
      boundBone {boneName=dyn_charm_02}
      boxExtents=<10 10 10>
      constraintSetup {angularLimitsMax=<5 25 25>, angularLimitsMin=<-5 -25 -25>, angularTarget=<1 0 0>, linearAxesMax=<1 0 0>}
      localJointOffset=<0.54 0 0>
     }]
    planarLimits [{drivingBone {boneName=Melee_TwoHanded_Handle}, planeTransform {location=<5 0 -35>, rotation=<rot 0 90 90>}}]
    relativeSpaceBone {boneName=Melee_TwoHanded_Handle}
    sphericalLimits [{limitRadius=0, drivingBone {boneName=Melee_TwoHanded_Handle}, sphereLocalOffset=<0 0 0>}]
   }
   n9 = Anim Dynamics - Bone: dyn_charm_03(ComponentPose=@n8, Alpha=1, ExternalForce="0,0,0") {
    angularBiasOverride=0.5
    angularDampingOverride=0.999
    angularSpringConstant=75
    bAngularSpring=true
    bOverrideAngularBias=true
    bOverrideAngularDamping=true
    bUseGravityOverride=true
    bUsePlanarLimit=false
    simulationSpace=BoneRelative
    alphaBoolBlend {blendOption=Linear}
    boundBone {boneName=dyn_charm_03}
    componentAppliedLinearAccClamp=<4000 4000 4000>
    componentLinearAccScale=<1 2 2>
    componentLinearVelScale=<1 2 2>
    gravityOverride {z=-980}
    physicsBodyDefinitions [{
      sphereCollisionRadius=7
      boundBone {boneName=dyn_charm_03}
      boxExtents=<10 10 5>
      constraintSetup {angularLimitsMax=<5 25 25>, angularLimitsMin=<-5 -25 -25>, angularTarget=<1 0 0>}
      localJointOffset=<1 0 0>
     }]
    planarLimits [{drivingBone {boneName=Melee_TwoHanded_Handle}, planeTransform {location=<-3 0 -35>, rotation=<rot 0 90 90>}}]
    relativeSpaceBone {boneName=Melee_TwoHanded_Handle}
    sphericalLimits [{limitRadius=100, limitType=Inner, drivingBone {boneName=Melee_TwoHanded_Handle}, sphereLocalOffset=<-100 10 -35>}]
   }
   Output Pose(Result=Blend Poses by bool(BlendPose_0=Component To Local(ComponentPose=@n5), BlendPose_1=Component To Local(ComponentPose=@n9)) {blendType=Linear})
   Save cached pose 'InitPose'(Pose=Input Pose()) {cacheName=InitPose}
  ]}
 EventGraph {nodes [
   Event BlueprintUpdateAnimation()
   > n1 = Gravity Override State Function DEPRECATED(CharacterPartAnimInstance=Self-Reference(), GameTimeInSeconds=GetGameTimeInSeconds(), GravityOverrideParams=Get Gravity Override Params())
   > Set GravityOverride(GravityOverride=@n1)
   TryGetPawnOwner()
  ]}
]
variables [
 Gravity Override Params {type=GravityOverrideParamsStruct}
 GravityOverride {type=Vector}
]
